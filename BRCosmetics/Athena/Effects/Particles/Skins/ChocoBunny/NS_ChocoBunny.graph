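# Niagara cosmetic effect for the ChocoBunny skin that throws small bursts of purple icing and multicolored sprinkles from the character's skeletal mesh.
# Left and right emitters spawn at two sampled bones of User.SkeletalMesh, inherit capped horizontal velocity, then launch in cones, fall and shrink.
# A left-only splash sprite expands and fades.
# Icing uses mesh particles, with _LOW sprite counterparts, and the splash has its own _LOW variant.
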NiagaraSystem /BRCosmetics/Athena/Effects/Particles/Skins/ChocoBunny/NS_ChocoBunny
dependencies {
 usedDataInterfaces [&NiagaraDataInterfaceVector2DCurve, &NiagaraDataInterfaceCurve]
 usedDynamicInputs []
 usedModules [&SystemState]
}
emitters {
 IcingSplash_Left {
  emitterUpdateScript [
   &EmitterState()
   &SpawnBurst_Instantaneous(Spawn Count=2, Use Spawn Probability=true)
  ]
  particleSpawnScript [
   /Niagara/Modules/Spawn/Initialization/InitializeParticle(
    Color=<rgba 19 19 19 1>
    Lifetime=0.4
    Module.Write Color=true
    Module.Write Lifetime=true
    Module.Write Mass=false
    Module.Write Material Random=true
    Module.Write Position=false
    Module.Write Scale=false
    Module.Write SpriteRotation=true
    Module.Write SpriteSize=true
    Position=&SimulationPosition()
    Sprite Rotation=&UniformRangedFloat(Minimum=0, Maximum=360)
    Sprite Size=&Vector2DFromFloat(Value=&UniformRangedFloat(Minimum=28, Maximum=32))
   )
   &SampleSkeletalMeshSkeleton(
    Mesh:NiagaraDataInterfaceSkeletalMesh=$User.SkeletalMesh
    Specific Bone Index:NiagaraInt32=&MakeIntFromFloat(Float=0)
   )
   /Niagara/Modules/Spawn/Location/SkeletalMeshSkeletonLocation(Module.Inherit Velocity=false, Module.Use Sampled Bone as Ribbon ID=true, Sampled Position Offset:Vector3f=<4 5 13>)
  ]
  particleUpdateScript [
   &AddVelocity(Coordinate Space=::Simulation, Velocity=<0 -8 -10>)
   &ParticleState()
   &SolveForcesAndVelocity(Clamp Velocity=false, Limit Acceleration=false)
   &ScaleSpriteSize(Scale Factor=&Vector2DFromCurve(Vector2Curve=DataInterface() {propertyValues {XCurve {keys [<at 0 0.4>,<at 1 3>]}, YCurve {keys [<at 0 0.4>,<at 1 3>]}}}, ScaleCurve=<1 1>))
   &ScaleColor(
    Scale Alpha=&FloatFromCurve(FloatCurve=DataInterface() {propertyValues {Curve {keys [
         <at -0.0542 -1>
         {arriveTangent=1.729, interpMode=RCIM_Cubic, leaveTangent=1.729, time=0.0002, value=0.5}
         {arriveTangent=-1.4986, interpMode=RCIM_Cubic, leaveTangent=-1.4986, time=0.533, value=0.0152}
         <at 1.0011 -1 RCIM_Cubic>
        ]}}})
    Scale RGB=&MakeVectorFromLinearColorRGB(Color=<rgba 0.6636 0.3392 0.9323 1>)
    ScaleA=true
    ScaleRGB=true
   )
  ]
  rendererClasses [&NiagaraSpriteRendererProperties]
  renderers [{rendererClass=&NiagaraSpriteRendererProperties}]
 }
 IcingSplash_Left_LOW {
  emitterUpdateScript [
   &EmitterState()
   &SpawnBurst_Instantaneous(Spawn Count=2, Use Spawn Probability=true)
  ]
  particleSpawnScript [
   &InitializeParticle(
    Color=<rgba 0.2847 0.1455 0.4 1>
    Lifetime=0.4
    Mass=&UniformRangedFloat(Minimum=0.75, Maximum=1.75)
    Module.Write Color=true
    Module.Write Lifetime=true
    Module.Write Mass=true
    Module.Write Material Random=true
    Module.Write Position=false
    Module.Write Scale=false
    Module.Write SpriteRotation=true
    Module.Write SpriteSize=true
    Position=&SimulationPosition()
    Sprite Rotation=&UniformRangedFloat(Minimum=0, Maximum=360)
    Sprite Size=&Vector2DFromFloat(Value=&UniformRangedFloat(Minimum=28, Maximum=32))
   )
   &SampleSkeletalMeshSkeleton(
    Mesh:NiagaraDataInterfaceSkeletalMesh=$User.SkeletalMesh
    Specific Bone Index:NiagaraInt32=&MakeIntFromFloat(Float=0)
   )
   &SkeletalMeshSkeletonLocation(Module.Inherit Velocity=false, Module.Use Sampled Bone as Ribbon ID=true, Sampled Position Offset:Vector3f=<4 5 13>)
  ]
  particleUpdateScript [
   &AddVelocity(Coordinate Space=::Simulation, Velocity=<0 -10 -10>)
   &ParticleState()
   &SolveForcesAndVelocity(Clamp Velocity=false, Limit Acceleration=false)
   &ScaleSpriteSize(Scale Factor=&Vector2DFromCurve(Vector2Curve=DataInterface() {propertyValues {XCurve {keys [<at 0 0.4>,<at 1 3>]}, YCurve {keys [<at 0 0.4>,<at 1 3>]}}}, ScaleCurve=<1 1>))
   &ScaleColor(
    Scale Alpha=&FloatFromCurve(FloatCurve=DataInterface() {propertyValues {Curve {keys [
         <at -0.0542 -1>
         {arriveTangent=1.729, interpMode=RCIM_Cubic, leaveTangent=1.729, time=0.0002, value=0.5}
         {arriveTangent=-1.4986, interpMode=RCIM_Cubic, leaveTangent=-1.4986, time=0.533, value=0.0152}
         <at 1.0011 -1 RCIM_Cubic>
        ]}}})
    ScaleA=true
    ScaleRGB=true
   )
  ]
  rendererClasses [&NiagaraSpriteRendererProperties]
  renderers [{rendererClass=&NiagaraSpriteRendererProperties}]
 }
 Icing_Left {
  emitterUpdateScript [
   &EmitterState()
   &SpawnBurst_Instantaneous(Spawn Count=2, Spawn Time=0.025, Use Spawn Probability=true)
  ]
  particleSpawnScript [
   &InitializeParticle(
    Color=<rgba 0.4065 0.2078 0.5711 1>
    Lifetime=&UniformRangedFloat(Minimum=0.5, Maximum=0.75)
    Mass=&UniformRangedFloat(Minimum=0.75, Maximum=1.75)
    Mesh Scale=&MakeVector(
     X=&UniformRangedFloat(Minimum=1.25, Maximum=2.75)
     Y=&UniformRangedFloat(Minimum=1.25, Maximum=2.75)
     Z=&UniformRangedFloat(Minimum=1, Maximum=2.75)
    )
    Module.Write Color=true
    Module.Write Lifetime=true
    Module.Write Mass=false
    Module.Write Material Random=true
    Module.Write Position=false
    Module.Write Scale=true
    Module.Write SpriteRotation=false
    Module.Write SpriteSize=false
    Position=&SimulationPosition()
    Sprite Rotation=180
   )
   &SampleSkeletalMeshSkeleton(
    Mesh:NiagaraDataInterfaceSkeletalMesh=$User.SkeletalMesh
    Sampled Position Coordinate Space:ENiagaraCoordinateSpace=::World
    Sampled Rotation Coordinate Space:ENiagaraCoordinateSpace=::World
    Sampled Velocity Coordinate Space:ENiagaraCoordinateSpace=::World
    Specific Bone Index:NiagaraInt32=&MakeIntFromFloat(Float=0)
   )
   &SkeletalMeshSkeletonLocation(Module.Inherit Velocity=false, Module.Use Sampled Bone as Ribbon ID=true)
   /Game/Effects/Niagara/Modules/Position/FN_OffsetPositionByOrientationAxis(Offset Axis:Vector3f=<0 0 0>) {module=OffsetPositionByMeshOrientation}
   /Niagara/Modules/Spawn/Location/SphereLocation(
    Coordinate Space:ENiagaraCoordinateSpace=::Simulation
    MaskSpawnGroup=false
    Module.Override Randomness=false
    Offset=<2 -10 -3>
    Sphere Radius=2
    Surface Expansion Mode:ENiagaraExpansionMode=::NewEnumerator2
    Surface Only=true
   )
   &AddVelocityInCone(
    Cone Angle=75
    Cone Axis=<1 0.5 0.75>
    Module.Use Velocity Falloff On Cone Axis=false
    Velocity Distribution Along Cone Axis=1
    Velocity Strength=&UniformRangedFloat(Minimum=50, Maximum=100)
   )
   &AddVelocity(Velocity=&UniformRangedVector(Minimum=<0 0 -25>, Maximum=<0 0 125>))
   /Niagara/Modules/Spawn/Orientation/V2/InitialMeshOrientation(
    Orientation Vector=&UniformRangedVector(Minimum=<0 0 0>, Maximum=<1 1 1>)
    Rotation=&UniformRangedVector(Minimum=<0 0 0>, Maximum=<360 360 360>)
    Use Orientation Vector=true
    Use Rotation Vector=true
   )
   &InheritVelocity(Module.Limit Speed=true, Velocity Limit=500, Velocity Scale=<1 1 0>)
  ]
  particleUpdateScript [
   &AccelerationForce(Acceleration=&UniformRangedVector(Minimum=<0 0 -500>, Maximum=<0 0 -1000>))
   &Drag(Drag=1.5, Use Linear Drag=true, Use Rotational Drag=false)
   &ParticleState()
   &SolveForcesAndVelocity(Clamp Velocity=false, Limit Acceleration=false)
   &ScaleMeshSize(Scale Factor=&VectorFromFloat(Value=&FloatFromCurve(FloatCurve=DataInterface() {propertyValues {Curve {keys [<at 0.5 1>,<at 1.0011 0.002>]}}})))
  ]
  rendererClasses [&NiagaraMeshRendererProperties]
  renderers [{rendererClass=&NiagaraMeshRendererProperties, rendererIndex=0}]
 }
 Icing_Left_LOW {
  emitterUpdateScript [
   &EmitterState()
   &SpawnBurst_Instantaneous(Spawn Count=2, Spawn Time=0.025, Use Spawn Probability=true)
  ]
  particleSpawnScript [
   &InitializeParticle(
    Color=<rgba 0.4065 0.2078 0.5711 1>
    Lifetime=&UniformRangedFloat(Minimum=0.5, Maximum=0.75)
    Mass=&UniformRangedFloat(Minimum=0.75, Maximum=1.75)
    Module.Write Color=true
    Module.Write Lifetime=true
    Module.Write Mass=false
    Module.Write Material Random=true
    Module.Write Position=false
    Module.Write Scale=true
    Module.Write SpriteRotation=true
    Module.Write SpriteSize=true
    Position=&SimulationPosition()
    Sprite Rotation=&UniformRangedFloat(Minimum=-360, Maximum=360)
    Sprite Size=&MakeVector2D(
     X=&UniformRangedFloat(Minimum=5, Maximum=15)
     Y=&UniformRangedFloat(Minimum=15, Maximum=25)
    )
   )
   &SampleSkeletalMeshSkeleton(
    Mesh:NiagaraDataInterfaceSkeletalMesh=$User.SkeletalMesh
    Sampled Position Coordinate Space:ENiagaraCoordinateSpace=::World
    Sampled Rotation Coordinate Space:ENiagaraCoordinateSpace=::World
    Sampled Velocity Coordinate Space:ENiagaraCoordinateSpace=::World
    Specific Bone Index:NiagaraInt32=&MakeIntFromFloat(Float=0)
   )
   &SkeletalMeshSkeletonLocation(Module.Inherit Velocity=false, Module.Use Sampled Bone as Ribbon ID=true)
   &FN_OffsetPositionByOrientationAxis(Offset Axis:Vector3f=<0 0 0>) {module=OffsetPositionByMeshOrientation}
   &SphereLocation(
    Coordinate Space:ENiagaraCoordinateSpace=::Simulation
    MaskSpawnGroup=false
    Module.Override Randomness=false
    Offset=<2 -10 -3>
    Sphere Radius=2
    Surface Expansion Mode:ENiagaraExpansionMode=::NewEnumerator2
    Surface Only=true
   )
   &AddVelocityInCone(
    Cone Angle=75
    Cone Axis=<1 0.5 0.75>
    Module.Use Velocity Falloff On Cone Axis=false
    Velocity Distribution Along Cone Axis=1
    Velocity Strength=&UniformRangedFloat(Minimum=50, Maximum=150)
   )
   &AddVelocity(Velocity=&UniformRangedVector(Minimum=<0 0 -25>, Maximum=<0 0 125>))
   &InheritVelocity(Module.Limit Speed=true, Velocity Limit=500, Velocity Scale=<1 1 0>)
  ]
  particleUpdateScript [
   &AccelerationForce(Acceleration=&UniformRangedVector(Minimum=<0 0 -500>, Maximum=<0 0 -1000>))
   &Drag(Drag=1.5, Use Linear Drag=true, Use Rotational Drag=false)
   &ParticleState()
   &SolveForcesAndVelocity(Clamp Velocity=false, Limit Acceleration=false)
   &ScaleSpriteSize(Scale Factor=&Vector2DFromCurve(
     Vector2Curve=DataInterface() {propertyValues {XCurve {keys [<at -0.0011 1.0067>,<at 0.9989 0>]}, YCurve {keys [<at -0.0011 1.0067>,<at 0.9989 0>]}}}
     ScaleCurve=<1 1>
    ))
  ]
  rendererClasses [&NiagaraSpriteRendererProperties]
  renderers [{rendererClass=&NiagaraSpriteRendererProperties}]
 }
 Icing_Right {
  emitterUpdateScript [
   &EmitterState()
   &SpawnBurst_Instantaneous(Spawn Count=1, Spawn Time=0.025, Use Spawn Probability=true)
  ]
  particleSpawnScript [
   &InitializeParticle(
    Color=<rgba 0.4065 0.2078 0.5711 1>
    Lifetime=&UniformRangedFloat(Minimum=0.5, Maximum=0.75)
    Mass=&UniformRangedFloat(Minimum=0.75, Maximum=1.75)
    Mesh Scale=&MakeVector(
     X=&UniformRangedFloat(Minimum=1, Maximum=2.5)
     Y=&UniformRangedFloat(Minimum=1, Maximum=2.5)
     Z=&UniformRangedFloat(Minimum=0.9, Maximum=2.5)
    )
    Module.Write Color=true
    Module.Write Lifetime=true
    Module.Write Mass=false
    Module.Write Material Random=true
    Module.Write Position=false
    Module.Write Scale=true
    Module.Write SpriteRotation=false
    Module.Write SpriteSize=false
    Position=&SimulationPosition()
    Sprite Rotation=180
   )
   &SampleSkeletalMeshSkeleton(
    Mesh:NiagaraDataInterfaceSkeletalMesh=$User.SkeletalMesh
    Sampled Position Coordinate Space:ENiagaraCoordinateSpace=::World
    Sampled Rotation Coordinate Space:ENiagaraCoordinateSpace=::World
    Sampled Velocity Coordinate Space:ENiagaraCoordinateSpace=::World
    Specific Bone Index:NiagaraInt32=&MakeIntFromFloat(Float=1)
   )
   &SkeletalMeshSkeletonLocation(Module.Inherit Velocity=false, Module.Use Sampled Bone as Ribbon ID=true)
   &FN_OffsetPositionByOrientationAxis(Offset Axis:Vector3f=<0 0 0>) {module=OffsetPositionByMeshOrientation}
   &SphereLocation(
    Coordinate Space:ENiagaraCoordinateSpace=::Simulation
    Hemisphere Y=true
    Hemisphere Z=true
    MaskSpawnGroup=false
    Module.Override Randomness=false
    Sphere Radius=2
    Surface Expansion Mode:ENiagaraExpansionMode=::NewEnumerator2
    Surface Only=true
   )
   &AddVelocityInCone(
    Cone Angle=25
    Cone Axis=<-1 -0.25 1>
    Module.Use Velocity Falloff On Cone Axis=false
    Velocity Distribution Along Cone Axis=1
    Velocity Strength=&UniformRangedFloat(Minimum=50, Maximum=100)
   )
   &AddVelocity(Velocity=&UniformRangedVector(Minimum=<0 0 -25>, Maximum=<0 0 100>))
   &InitialMeshOrientation(
    Orientation Vector=&UniformRangedVector(Minimum=<0 0 0>, Maximum=<1 1 1>)
    Rotation=&UniformRangedVector(Minimum=<0 0 0>, Maximum=<360 360 360>)
    Use Orientation Vector=true
    Use Rotation Vector=true
   )
   &InheritVelocity(Module.Limit Speed=true, Velocity Limit=500, Velocity Scale=<1 1 0>)
  ]
  particleUpdateScript [
   &AccelerationForce(Acceleration=&UniformRangedVector(Minimum=<0 0 -500>, Maximum=<0 0 -1000>))
   &Drag(Drag=1.5, Use Linear Drag=true, Use Rotational Drag=false)
   &ParticleState()
   &SolveForcesAndVelocity(Clamp Velocity=false, Limit Acceleration=false)
   &ScaleMeshSize(Scale Factor=&VectorFromFloat(Value=&FloatFromCurve(FloatCurve=DataInterface() {propertyValues {Curve {keys [<at 0.5 1>,<at 1.0011 0.002>]}}})))
  ]
  rendererClasses [&NiagaraMeshRendererProperties]
  renderers [{rendererClass=&NiagaraMeshRendererProperties, rendererIndex=0}]
 }
 Icing_Right_LOW {
  emitterUpdateScript [
   &EmitterState()
   &SpawnBurst_Instantaneous(Spawn Count=1, Spawn Time=0.025, Use Spawn Probability=true)
  ]
  particleSpawnScript [
   &InitializeParticle(
    Color=<rgba 0.4065 0.2078 0.5711 1>
    Lifetime=&UniformRangedFloat(Minimum=0.5, Maximum=0.75)
    Mass=&UniformRangedFloat(Minimum=0.75, Maximum=1.75)
    Module.Write Color=true
    Module.Write Lifetime=true
    Module.Write Mass=false
    Module.Write Material Random=true
    Module.Write Position=false
    Module.Write Scale=true
    Module.Write SpriteRotation=true
    Module.Write SpriteSize=true
    Position=&SimulationPosition()
    Sprite Rotation=&UniformRangedFloat(Minimum=-360, Maximum=360)
    Sprite Size=&MakeVector2D(
     X=&UniformRangedFloat(Minimum=5, Maximum=15)
     Y=&UniformRangedFloat(Minimum=15, Maximum=25)
    )
   )
   &SampleSkeletalMeshSkeleton(
    Mesh:NiagaraDataInterfaceSkeletalMesh=$User.SkeletalMesh
    Sampled Position Coordinate Space:ENiagaraCoordinateSpace=::World
    Sampled Rotation Coordinate Space:ENiagaraCoordinateSpace=::World
    Sampled Velocity Coordinate Space:ENiagaraCoordinateSpace=::World
    Specific Bone Index:NiagaraInt32=&MakeIntFromFloat(Float=1)
   )
   &SkeletalMeshSkeletonLocation(Module.Inherit Velocity=false, Module.Use Sampled Bone as Ribbon ID=true)
   &FN_OffsetPositionByOrientationAxis(Offset Axis:Vector3f=<0 0 0>) {module=OffsetPositionByMeshOrientation}
   &SphereLocation(
    Coordinate Space:ENiagaraCoordinateSpace=::Simulation
    Hemisphere Y=true
    Hemisphere Z=true
    MaskSpawnGroup=false
    Module.Override Randomness=false
    Offset=<5 -5 0>
    Sphere Radius=2
    Surface Expansion Mode:ENiagaraExpansionMode=::NewEnumerator2
    Surface Only=true
   )
   &AddVelocityInCone(
    Cone Angle=25
    Cone Axis=<-1 -0.25 1>
    Module.Use Velocity Falloff On Cone Axis=false
    Velocity Distribution Along Cone Axis=1
    Velocity Strength=&UniformRangedFloat(Minimum=50, Maximum=150)
   )
   &AddVelocity(Velocity=&UniformRangedVector(Minimum=<0 0 -25>, Maximum=<0 0 100>))
   &InheritVelocity(Module.Limit Speed=true, Velocity Limit=500, Velocity Scale=<1 1 0>)
  ]
  particleUpdateScript [
   &AccelerationForce(Acceleration=&UniformRangedVector(Minimum=<0 0 -500>, Maximum=<0 0 -1000>))
   &Drag(Drag=1.5, Use Linear Drag=true, Use Rotational Drag=false)
   &ParticleState()
   &SolveForcesAndVelocity(Clamp Velocity=false, Limit Acceleration=false)
   &ScaleSpriteSize(Scale Factor=&Vector2DFromCurve(
     Vector2Curve=DataInterface() {propertyValues {XCurve {keys [<at -0.0011 1.0067>,<at 0.9988 0>]}, YCurve {keys [<at -0.0011 1.0067>,<at 0.9988 0>]}}}
     ScaleCurve=<1 1>
    ))
  ]
  rendererClasses [&NiagaraSpriteRendererProperties]
  renderers [{rendererClass=&NiagaraSpriteRendererProperties}]
 }
 Sprinkles {
  emitterUpdateScript [
   &EmitterState()
   &SpawnBurst_Instantaneous(Spawn Count=2, Use Spawn Probability=true)
  ]
  particleSpawnScript [
   &InitializeParticle(
    Color=&LerpLinearColors(
     StartColor=&LerpLinearColors(
      StartColor=<rgba 0.8963 0.6299 0.2054 1>
      EndColor=<rgba 0.2101 0.4462 0.8963 1>
      LerpFactor=&UniformAOrBFloat(A=0)
     )
     EndColor=&LerpLinearColors(
      StartColor=<rgba 0 0.3419 0.5647 1>
      EndColor=<rgba 0.624 0.0908 0.624 1>
      LerpFactor=&UniformAOrBFloat(A=0)
     )
     LerpFactor=&UniformAOrBFloat(A=1, B=0)
    )
    Lifetime=&UniformRangedFloat(Minimum=0.5, Maximum=0.75)
    Module.Write Color=true
    Module.Write Lifetime=true
    Module.Write Mass=false
    Module.Write Material Random=true
    Module.Write Position=false
    Module.Write Scale=false
    Module.Write SpriteRotation=true
    Module.Write SpriteSize=true
    Position=&SimulationPosition()
    Sprite Rotation=&UniformRangedFloat(Minimum=-360, Maximum=360)
    Sprite Size=&UniformRangedVector2D(Minimum=<2 10>, Maximum=<2.5 12>)
   )
   &SampleSkeletalMeshSkeleton(
    Mesh:NiagaraDataInterfaceSkeletalMesh=$User.SkeletalMesh
    Sampled Position Coordinate Space:ENiagaraCoordinateSpace=::World
    Sampled Rotation Coordinate Space:ENiagaraCoordinateSpace=::World
    Sampled Velocity Coordinate Space:ENiagaraCoordinateSpace=::World
    Specific Bone Index:NiagaraInt32=&MakeIntFromFloat(Float=0)
   )
   &SkeletalMeshSkeletonLocation(Module.Inherit Velocity=false, Module.Use Sampled Bone as Ribbon ID=true)
   &FN_OffsetPositionByOrientationAxis(Offset Axis:Vector3f=<0 0 0>) {module=OffsetPositionByMeshOrientation}
   &SphereLocation(
    Coordinate Space:ENiagaraCoordinateSpace=::Simulation
    MaskSpawnGroup=false
    Module.Override Randomness=false
    Offset=<2 0 -2>
    Sphere Radius=4
    Surface Expansion Mode:ENiagaraExpansionMode=::NewEnumerator2
    Surface Only=true
   )
   &AddVelocityInCone(
    Cone Angle=75
    Cone Axis=<1 0.5 1>
    Module.Use Velocity Falloff On Cone Axis=false
    Velocity Distribution Along Cone Axis=1
    Velocity Strength=&UniformRangedFloat(Minimum=50, Maximum=75)
   )
   &AddVelocity(Velocity=&UniformRangedVector(Minimum=<0 0 -25>, Maximum=<0 0 75>))
   &InheritVelocity(Module.Limit Speed=true, Velocity Limit=500, Velocity Scale=<1 1 0>)
  ]
  particleUpdateScript [
   &AccelerationForce(Acceleration=&UniformRangedVector(Minimum=<0 0 -500>, Maximum=<0 0 -1000>))
   &ParticleState()
   &SolveForcesAndVelocity(Clamp Velocity=false, Limit Acceleration=false)
   &ScaleSpriteSize(Scale Factor=&Vector2DFromCurve(
     Vector2Curve=DataInterface() {propertyValues {XCurve {keys [<at 0.75 1>,<at 0.9989 0>]}, YCurve {keys [<at 0.75 1>,<at 0.9989 0>]}}}
     ScaleCurve=<1 1>
    ))
   &SpriteRotationRate(Rotation Rate=&UniformRangedFloat(Minimum=180, Maximum=360))
  ]
  rendererClasses [&NiagaraSpriteRendererProperties]
  renderers [{rendererClass=&NiagaraSpriteRendererProperties}]
 }
 Sprinkles_Right {
  emitterUpdateScript [
   &EmitterState()
   &SpawnBurst_Instantaneous(Spawn Count=2, Use Spawn Probability=true)
  ]
  particleSpawnScript [
   &InitializeParticle(
    Color=&LerpLinearColors(
     StartColor=&LerpLinearColors(
      StartColor=<rgba 0.624 0.0908 0.624 1>
      EndColor=<rgba 0 0.3419 0.5647 1>
      LerpFactor=&UniformAOrBFloat(A=0)
     )
     EndColor=&LerpLinearColors(
      StartColor=<rgba 0.2101 0.4462 0.8963 1>
      EndColor=<rgba 0.8963 0.6299 0.2054 1>
      LerpFactor=&UniformAOrBFloat(A=0)
     )
     LerpFactor=&UniformAOrBFloat(A=1, B=0)
    )
    Lifetime=&UniformRangedFloat(Minimum=0.5, Maximum=0.75)
    Mass=&UniformRangedFloat(Minimum=0.75, Maximum=1.75)
    Module.Write Color=true
    Module.Write Lifetime=true
    Module.Write Mass=false
    Module.Write Material Random=true
    Module.Write Position=false
    Module.Write Scale=false
    Module.Write SpriteRotation=true
    Module.Write SpriteSize=true
    Position=&SimulationPosition()
    Sprite Rotation=&UniformRangedFloat(Minimum=-360, Maximum=360)
    Sprite Size=&UniformRangedVector2D(Minimum=<1.75 8>, Maximum=<2.25 10>)
   )
   &SampleSkeletalMeshSkeleton(
    Mesh:NiagaraDataInterfaceSkeletalMesh=$User.SkeletalMesh
    Sampled Position Coordinate Space:ENiagaraCoordinateSpace=::World
    Sampled Rotation Coordinate Space:ENiagaraCoordinateSpace=::World
    Sampled Velocity Coordinate Space:ENiagaraCoordinateSpace=::World
    Specific Bone Index:NiagaraInt32=&MakeIntFromFloat(Float=1)
   )
   &SkeletalMeshSkeletonLocation(Module.Inherit Velocity=false, Module.Use Sampled Bone as Ribbon ID=true)
   &FN_OffsetPositionByOrientationAxis(Offset Axis:Vector3f=<0 0 0>) {module=OffsetPositionByMeshOrientation}
   &SphereLocation(
    Coordinate Space:ENiagaraCoordinateSpace=::Simulation
    Hemisphere Y=true
    Hemisphere Z=true
    MaskSpawnGroup=false
    Module.Override Randomness=false
    Sphere Radius=2
    Surface Expansion Mode:ENiagaraExpansionMode=::NewEnumerator2
    Surface Only=true
   )
   &AddVelocityInCone(
    Cone Angle=25
    Cone Axis=<-1 -0.25 0.75>
    Module.Use Velocity Falloff On Cone Axis=false
    Velocity Distribution Along Cone Axis=1
    Velocity Strength=&UniformRangedFloat(Minimum=50, Maximum=75)
   )
   &AddVelocity(Velocity=&UniformRangedVector(Minimum=<0 0 -25>, Maximum=<0 0 75>))
   &InheritVelocity(Module.Limit Speed=true, Velocity Limit=500, Velocity Scale=<1 1 0>)
  ]
  particleUpdateScript [
   &AccelerationForce(Acceleration=&UniformRangedVector(Minimum=<0 0 -500>, Maximum=<0 0 -1000>))
   &ParticleState()
   &SolveForcesAndVelocity(Clamp Velocity=false, Limit Acceleration=false)
   &ScaleSpriteSize(Scale Factor=&Vector2DFromCurve(
     Vector2Curve=DataInterface() {propertyValues {XCurve {keys [<at 0.75 1>,<at 0.9989 0>]}, YCurve {keys [<at 0.75 1>,<at 0.9989 0>]}}}
     ScaleCurve=<1 1>
    ))
   &SpriteRotationRate(Rotation Rate=&UniformRangedFloat(Minimum=180, Maximum=360))
  ]
  rendererClasses [&NiagaraSpriteRendererProperties]
  renderers [{rendererClass=&NiagaraSpriteRendererProperties}]
 }
}
summary {
 systemName=NS_ChocoBunny
 emitters [
  {emitterName=Sprinkles, rendererClasses [&NiagaraSpriteRendererProperties]}
  {emitterName=Sprinkles_Right, rendererClasses [&NiagaraSpriteRendererProperties]}
  {emitterName=Icing_Left, rendererClasses [&NiagaraMeshRendererProperties]}
  {emitterName=Icing_Right, rendererClasses [&NiagaraMeshRendererProperties]}
  {emitterName=IcingSplash_Left, rendererClasses [&NiagaraSpriteRendererProperties]}
  {emitterName=Icing_Left_LOW, rendererClasses [&NiagaraSpriteRendererProperties]}
  {emitterName=Icing_Right_LOW, rendererClasses [&NiagaraSpriteRendererProperties]}
  {emitterName=IcingSplash_Left_LOW, rendererClasses [&NiagaraSpriteRendererProperties]}
 ]
 userVariables [User.SkeletalMesh {
   type=NiagaraDataInterfaceSkeletalMesh
   defaultValue {
    struct=NiagaraExt_VariableValue_DataInterface
    value {dataInterface=&self:NiagaraDataInterfaceSkeletalMesh_0}
   }
  }]
}
